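# Sound cue for the electric ceiling trap's zap, an Athena variant with trap-activation attenuation and a concurrency override that prevents new instances at the limit.
# A quality-level node picks one of four mixes, each layering a fire sound with a ready sound behind a delay node.
# The first randomizes three fire waves; the others save variety with one pitch-modulated wave and a fixed delay on the ready sound, the last two using its optimized wave.
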SoundCue /FNGameplayCues/Sounds/Fort_Traps/Electric_Trap/Electric_Trap_Zap_Athena
root=QualityLevel_0
nodes [
 {id=Delay_0, children [WavePlayer_7]}
 {id=Delay_1, children [WavePlayer_9], properties {delayMax=0.7, delayMin=0.7}}
 {id=Delay_2, children [WavePlayer_3], properties {delayMax=0.7, delayMin=0.7}}
 {id=Delay_4, children [WavePlayer_4], properties {delayMax=0.7, delayMin=0.7}}
 {id=Mixer_0, children [Random_0,Delay_0], properties {inputVolume [1,1]}}
 {id=Mixer_1, children [Modulator_0,Delay_1], properties {inputVolume [1,1]}}
 {id=Mixer_2, children [Modulator_4,Delay_2], properties {inputVolume [1,1]}}
 {id=Mixer_3, children [Modulator_5,Delay_4], properties {inputVolume [1,1]}}
 {id=Modulator_0, children [WavePlayer_8], properties {pitchMax=1.1, pitchMin=1}}
 {id=Modulator_4, children [WavePlayer_10], properties {pitchMax=1.1, pitchMin=1}}
 {id=Modulator_5, children [WavePlayer_12], properties {pitchMax=1.1, pitchMin=1}}
 {id=QualityLevel_0, children [Mixer_0,Mixer_1,Mixer_2,Mixer_3]}
 {id=Random_0, children [WavePlayer_0,WavePlayer_1,WavePlayer_2], properties {weights [1,1,1]}}
 {id=WavePlayer_0, properties {soundWaveAssetPtr=&/CeilingTrap_Electric_Fire_001}}
 {id=WavePlayer_1, properties {soundWaveAssetPtr=/Game/Sounds/Fort_Traps/Electric_Trap/CeilingTrap_Electric_Fire_002}}
 {id=WavePlayer_10, properties {soundWaveAssetPtr=&CeilingTrap_Electric_Fire_002}}
 {id=WavePlayer_12, properties {soundWaveAssetPtr=&CeilingTrap_Electric_Fire_002}}
 {id=WavePlayer_2, properties {soundWaveAssetPtr=&/CeilingTrap_Electric_Fire_004}}
 {id=WavePlayer_3, properties {soundWaveAssetPtr=&/CeilingTrap_Electric_Ready_002_opt}}
 {id=WavePlayer_4, properties {soundWaveAssetPtr=&CeilingTrap_Electric_Ready_002_opt}}
 {id=WavePlayer_7, properties {soundWaveAssetPtr=&/CeilingTrap_Electric_Ready_002}}
 {id=WavePlayer_8, properties {soundWaveAssetPtr=&CeilingTrap_Electric_Fire_002}}
 {id=WavePlayer_9, properties {soundWaveAssetPtr=&CeilingTrap_Electric_Ready_002}}
]
properties {
 attenuationSettings=/Game/Sounds/Attenuation/Fort_Traps_Activate_Athena_Att
 bOverrideConcurrency=true
 maxDistance=3500
 soundClassObject=&Fort_Traps
 volumeMultiplier=0.5
 attenuationOverrides {falloffDistance=3200, attenuationShapeExtents {x=800}}
 concurrencyOverrides {resolutionRule=PreventNew}
}
